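# Control Rig for the MelodyUrchin face accessory part that adjusts nose ring, hair and braid master bones to follow face and head motion.
# A partial-weight parent constraint pulls the nose ring toward the nose bone in translation and scale only.
# A sequence then runs three spherical pose readers on head orientation relative to spine_05.
# Their outputs weight corrective transform offsets: one for hair and both braids, the others for each braid separately.
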ControlRigBlueprint /BRCosmetics/Characters/Player/Female/Medium/Bodies/F_MED_MelodyUrchin/Meshes/Parts/F_MED_MelodyUrchin_FaceAcc_CtrlRig
graphs [
 RigVMFunctionLibrary {links [], nodes []}
 RigVMFunction_Sequence_1_ContainedGraph {nodes [
   n1 = RigVMFunctionEntryNode()
   RigVMFunction_Sequence = &RigVMFunction_Sequence(ExecuteContext=@n1.ExecuteContext)
   n2 = &RigVMFunction_Sequence(ExecuteContext=@RigVMFunction_Sequence.B)
   n3 = &RigVMFunction_Sequence(ExecuteContext=@n2.B)
   RigVMFunctionReturnNode(A=@RigVMFunction_Sequence.A, B=@n2.A, C=@n3.A, D=@n3.B)
  ]}
 RigVMModel {nodes [
   RigUnit_BeginExecution = &RigUnit_BeginExecution()
   &RigUnit_ParentConstraint(
    AdvancedSettings=(InterpolationType=Average,RotationOrderForFilter=XZY)
    Child=(Type=Bone,Name="dyn_nosering_MSTR")
    Filter=(TranslationFilter=(bX=True,bY=True,bZ=True),RotationFilter=(bX=false,bY=false,bZ=false),ScaleFilter=(bX=True,bY=True,bZ=True))
    Parents=((Item=(Type=Bone,Name="FACIAL_C_Nose"),Weight=1))
    Weight=0.3
    ExecutePin=@RigUnit_BeginExecution.ExecutePin
   )
   > RigVMFunction_Sequence = RigVMAggregateNode()
   n1 = &RigUnit_SphericalPoseReader(
    ActiveRegionScaleFactors=(PositiveWidth=3.5,NegativeWidth=0.75,PositiveHeight=3,NegativeHeight=3)
    ActiveRegionSize=0.1
    Debug=(bDrawDebug=true,bDraw2D=false,bDrawLocalAxes=false,DebugScale=25,DebugSegments=20,DebugThickness=0.25)
    DriverAxis=(X=1,Y=0,Z=0)
    DriverItem=(Type=Bone,Name="head")
    FalloffRegionScaleFactors=(PositiveWidth=1,NegativeWidth=1,PositiveHeight=1,NegativeHeight=1)
    FalloffSize=0.45
    FlipHeightScaling=true
    FlipWidthScaling=true
    OptionalParentItem=(Type=Bone,Name="spine_05")
    RotationOffset=(X=-90,Y=0,Z=0)
    ExecutePin=@RigVMFunction_Sequence.A
   )
   > &RigUnit_ModifyTransforms(
    ItemToModify=((Item=(Type=Bone,Name="dyn_hair_MSTR"),Transform=(Rotation=(X=0,Y=0,Z=-0.5,W=0.866),Translation=(X=-8,Y=-8,Z=0),Scale3D=(X=1,Y=1,Z=1))))
    WeightMinimum=0
    Weight=@n1.OutputParam
   )
   > &RigUnit_ModifyTransforms(
    ItemToModify=((Item=(Type=Bone,Name="dyn_braid_l_MSTR"),Transform=(Rotation=(X=0.0087,Y=0,Z=0,W=1),Translation=(X=-2,Y=-2,Z=0),Scale3D=(X=1,Y=1,Z=1))))
    WeightMinimum=0
    Weight=@n1.OutputParam
   )
   > &RigUnit_ModifyTransforms(
    ItemToModify=((Item=(Type=Bone,Name="dyn_braid_r_MSTR"),Transform=(Rotation=(X=0,Y=0,Z=0,W=1),Translation=(X=-2,Y=-2,Z=0),Scale3D=(X=1,Y=1,Z=1))))
    WeightMinimum=0
    Weight=@n1.OutputParam
   )
   n2 = &RigUnit_SphericalPoseReader(
    ActiveRegionScaleFactors=(PositiveWidth=3,NegativeWidth=0.75,PositiveHeight=3,NegativeHeight=3)
    ActiveRegionSize=0.1
    Debug=(bDrawDebug=true,bDraw2D=false,bDrawLocalAxes=false,DebugScale=25,DebugSegments=20,DebugThickness=0.25)
    DriverAxis=(X=1,Y=0,Z=0)
    DriverItem=(Type=Bone,Name="head")
    FalloffRegionScaleFactors=(PositiveWidth=0.1,NegativeWidth=0.1,PositiveHeight=0.1,NegativeHeight=0.1)
    FalloffSize=1
    FlipHeightScaling=true
    FlipWidthScaling=true
    OptionalParentItem=(Type=Bone,Name="spine_05")
    RotationOffset=(X=0,Y=-20,Z=0)
    ExecutePin=@RigVMFunction_Sequence.B
   )
   > &RigUnit_ModifyTransforms(
    ItemToModify=((Item=(Type=Bone,Name="dyn_braid_r_MSTR"),Transform=(Rotation=(X=0,Y=0,Z=0,W=1),Translation=(X=-3,Y=-1,Z=0),Scale3D=(X=1,Y=1,Z=1))))
    WeightMinimum=0
    Weight=@n2.OutputParam
   )
   n3 = &RigUnit_SphericalPoseReader(
    ActiveRegionScaleFactors=(PositiveWidth=3,NegativeWidth=0.75,PositiveHeight=3,NegativeHeight=3)
    ActiveRegionSize=0.1
    Debug=(bDrawDebug=true,bDraw2D=false,bDrawLocalAxes=false,DebugScale=25,DebugSegments=20,DebugThickness=0.25)
    DriverAxis=(X=1,Y=0,Z=0)
    DriverItem=(Type=Bone,Name="head")
    FalloffRegionScaleFactors=(PositiveWidth=0.1,NegativeWidth=0.1,PositiveHeight=0.1,NegativeHeight=0.1)
    FalloffSize=1
    FlipHeightScaling=true
    FlipWidthScaling=true
    OptionalParentItem=(Type=Bone,Name="spine_05")
    RotationOffset=(X=0,Y=160,Z=180)
    ExecutePin=@RigVMFunction_Sequence.C
   )
   > &RigUnit_ModifyTransforms(
    ItemToModify=((Item=(Type=Bone,Name="dyn_braid_l_MSTR"),Transform=(Rotation=(X=0,Y=0,Z=0,W=1),Translation=(X=-3,Y=-1,Z=0),Scale3D=(X=1,Y=1,Z=1))))
    WeightMinimum=0
    Weight=@n3.OutputParam
   )
  ]}
]
hierarchy [
 root {loc=<0 0 0>, rot=<0 0 0 1>}
 pelvis {parent=root, loc=<0 -0.5301 98.0571>, rot=<0.0127 0.707 -0.0127 -0.707>}
 spine_01 {parent=pelvis, loc=<6.1208 -0.9328 0>, rot=<0 0 0.034 -0.9994>}
 spine_02 {parent=spine_01, loc=<6.5996 -0.2095 0>, rot=<0 0 0.0238 -0.9997>}
 spine_03 {parent=spine_02, loc=<6.4914 -0.1237 0>, rot=<0 0 0.0149 -0.9999>}
 spine_04 {parent=spine_03, loc=<6.9598 0.0794 0>, rot=<0 0 -0.0132 -0.9999>}
 spine_05 {parent=spine_04, loc=<6.9987 0 0>, rot=<0 0 -0.0015 -1>}
 neck_01 {parent=spine_05, loc=<11.3138 0.1529 0>, rot=<0 0 -0.2003 -0.9797>}
 neck_02 {parent=neck_01, loc=<4.5627 0 0>, rot=<0 0 0.0504 -0.9987>}
 head {parent=neck_02, loc=<5.3914 0 0>, rot=<0 0 0.115 -0.9934>}
 dyn_hair_MSTR {parent=head, loc=<-3.9641 -6.8632 0>, rot=<0.2026 -0.9793 0 0>}
 dyn_hair_01 {parent=dyn_hair_MSTR, loc=<0 0 0>, rot=<0 0 0 1>}
 dyn_hair_02 {parent=dyn_hair_01, loc=<3.7389 -0.0829 0>, rot=<0 0 0.066 -0.9978>}
 dyn_hair_03 {parent=dyn_hair_02, loc=<4.097 0.0972 0>, rot=<0 0 -0.0775 -0.997>}
 dyn_hair_04 {parent=dyn_hair_03, loc=<4.4962 0.0968 0>, rot=<0 0 -0.0772 -0.997>}
 dyn_braid_r_MSTR {parent=head, loc=<2.6308 -6.0472 4.7264>, rot=<0.2414 -0.9582 0.1464 0.0454>}
 dyn_braid_r_01 {parent=dyn_braid_r_MSTR, loc=<0 0 0>, rot=<0 0 0 1>}
 dyn_braid_r_02 {parent=dyn_braid_r_01, loc=<7.4856 0 0>, rot=<0 0 0 -1>}
 dyn_braid_r_03 {parent=dyn_braid_r_02, loc=<3.9421 0 0>, rot=<0 0 0 -1>}
 dyn_braid_r_04 {parent=dyn_braid_r_03, loc=<3.6589 0 0>, rot=<-0.0054 -0.0109 -0.0696 -0.9975>}
 dyn_braid_r_05 {parent=dyn_braid_r_04, loc=<4.5239 -0.0001 0>, rot=<0.0078 0.1031 -0.085 -0.991>}
 dyn_braid_r_06 {parent=dyn_braid_r_05, loc=<7.2766 -0.0001 0>, rot=<-0.0785 0.021 -0.0529 -0.9953>}
 dyn_braid_l_MSTR {parent=head, loc=<2.6308 -6.0472 -4.7264>, rot=<0.2414 -0.9582 -0.1464 -0.0454>}
 dyn_braid_l_01 {parent=dyn_braid_l_MSTR, loc=<0 0 0>, rot=<0 0 0 1>}
 dyn_braid_l_02 {parent=dyn_braid_l_01, loc=<7.4856 0 0>, rot=<0 0 0 -1>}
 dyn_braid_l_03 {parent=dyn_braid_l_02, loc=<3.9421 0 0>, rot=<0 0 0 -1>}
 dyn_braid_l_04 {parent=dyn_braid_l_03, loc=<3.6589 0 0>, rot=<0.0054 0.0109 -0.0696 -0.9975>}
 dyn_braid_l_05 {parent=dyn_braid_l_04, loc=<4.524 0 0>, rot=<-0.0078 -0.1031 -0.085 -0.991>}
 dyn_braid_l_06 {parent=dyn_braid_l_05, loc=<7.2768 0 0>, rot=<0.0785 -0.021 -0.0529 -0.9953>}
 dyn_bang_r_MSTR {parent=head, loc=<6.3993 6.3768 5.5394>, rot=<0.2066 -0.4353 0.8721 0.085>}
 dyn_bang_r_01 {parent=dyn_bang_r_MSTR, loc=<0 0 0>, rot=<0 0 0 1>}
 dyn_bang_r_02 {parent=dyn_bang_r_01, loc=<3.3296 0 0>, rot=<-0.0157 0.0309 -0.1582 -0.9868>}
 dyn_bang_r_03 {parent=dyn_bang_r_02, loc=<3.5274 0 0>, rot=<-0.0173 0.0192 -0.0783 -0.9966>}
 dyn_bang_r_04 {parent=dyn_bang_r_03, loc=<4.1145 0 0>, rot=<-0.018 -0.1767 0.0365 -0.9834>}
 dyn_bang_c_MSTR {parent=head, loc=<11.6358 7.576 0.1075>, rot=<0.0547 -0.0804 -0.7427 -0.6626>}
 dyn_bang_c_01 {parent=dyn_bang_c_MSTR, loc=<0 0 0>, rot=<0 0 0 1>}
 dyn_bang_c_02 {parent=dyn_bang_c_01, loc=<1.8791 0 0>, rot=<0.0192 0.0006 -0.3786 -0.9253>}
 dyn_bang_c_03 {parent=dyn_bang_c_02, loc=<2.1124 0 0>, rot=<-0.0291 0.0076 -0.315 -0.9486>}
 dyn_bang_c_04 {parent=dyn_bang_c_03, loc=<3.3212 0 0>, rot=<-0.0007 -0.0528 -0.0581 -0.9969>}
 dyn_bang_c_05 {parent=dyn_bang_c_04, loc=<2.7767 0 0>, rot=<0.008 -0.1346 0.1665 -0.9768>}
 dyn_bang_l_f_MSTR {parent=head, loc=<10.6886 7.9789 -2.1115>, rot=<-0.1702 0.2242 0.8481 0.4488>}
 dyn_bang_l_f_01 {parent=dyn_bang_l_f_MSTR, loc=<0 0 0>, rot=<0 0 0 1>}
 dyn_bang_l_f_02 {parent=dyn_bang_l_f_01, loc=<1.9514 0 0>, rot=<-0.066 -0.1333 -0.311 -0.9387>}
 dyn_bang_l_f_03 {parent=dyn_bang_l_f_02, loc=<2.779 0 0>, rot=<-0.0026 -0.1667 -0.1 -0.9809>}
 dyn_bang_l_f_04 {parent=dyn_bang_l_f_03, loc=<3.4324 0 0>, rot=<0.0424 -0.0562 -0.1796 -0.9812>}
 dyn_bang_l_f_05 {parent=dyn_bang_l_f_04, loc=<3.0715 0 0>, rot=<0.1035 0.3073 -0.0585 -0.9442>}
 dyn_bang_l_b_MSTR {parent=head, loc=<6.7625 7.5619 -4.4343>, rot=<-0.0609 0.3821 0.9212 0.0414>}
 dyn_bang_l_b_01 {parent=dyn_bang_l_b_MSTR, loc=<0 0 0>, rot=<0 0 0 1>}
 dyn_bang_l_b_02 {parent=dyn_bang_l_b_01, loc=<3.1376 0 0>, rot=<0.0012 -0.01 -0.0794 -0.9968>}
 dyn_bang_l_b_03 {parent=dyn_bang_l_b_02, loc=<3.046 0 0>, rot=<-0.0099 0.2371 -0.0488 -0.9702>}
 dyn_bang_l_MSTR {parent=head, loc=<5.0207 2.0813 -7.3391>, rot=<-0.1968 0.6639 0.7168 -0.0819>}
 dyn_bang_l_01 {parent=dyn_bang_l_MSTR, loc=<0 0 0>, rot=<0 0 0 1>}
 dyn_bang_l_02 {parent=dyn_bang_l_01, loc=<1.2655 0 0>, rot=<0.0154 0.2733 -0.1038 -0.9562>}
 dyn_earring_r_MSTR {parent=head, loc=<0.3924 0.0847 6.2564>, rot=<0.0262 -0.8254 0.5636 -0.0171>}
 dyn_earring_r_01 {parent=dyn_earring_r_MSTR, loc=<0 0 0>, rot=<0 0 0 1>}
 dyn_earring_l_MSTR {parent=head, loc=<0.3924 0.0847 -6.2564>, rot=<0.0273 -0.862 -0.5059 0.0153>}
 dyn_earring_l_01 {parent=dyn_earring_l_MSTR, loc=<0 0 0>, rot=<0 0 0 1>}
 FACIAL_C_FacialRoot {parent=head, loc=<1.6805 0.2827 0>, rot=<0.5021 -0.4979 -0.5021 -0.4979>}
 FACIAL_C_Nose {parent=FACIAL_C_FacialRoot, loc=<-0.0009 -0.2574 8.2153>, rot=<-0.162 0 0 -0.9868>}
 FACIAL_C_NoseLower {parent=FACIAL_C_Nose, loc=<0.0012 2.3787 -0.3406>, rot=<0.1312 0 0 -0.9914>}
 dyn_nosering_MSTR {parent=FACIAL_C_NoseLower, loc=<0.1116 -0.5945 -0.1581>, rot=<0.5797 0.4104 0.6275 0.319>}
 dyn_nosering_01 {parent=dyn_nosering_MSTR, loc=<0 0 0>, rot=<0 0 0 1>}
 clavicle_l {parent=spine_05, loc=<5.4831 -0.1053 -3.2946>, rot=<0.0465 -0.7289 -0.009 -0.6829>}
 upperarm_l {parent=clavicle_l, loc=<9.8658 -1.2136 -0.4105>, rot=<0.0164 -0.3245 -0.0269 -0.9454>}
 clavicle_r {parent=spine_05, loc=<5.4831 -0.1053 3.2946>, rot=<0.7289 0.0465 -0.6829 0.009>}
 upperarm_r {parent=clavicle_r, loc=<-9.8658 1.2136 0.4105>, rot=<0.0164 -0.3245 -0.0269 -0.9454>}
]
