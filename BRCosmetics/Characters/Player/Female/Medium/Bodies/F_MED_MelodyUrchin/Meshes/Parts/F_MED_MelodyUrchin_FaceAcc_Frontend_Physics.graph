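# Physics asset for the Frontend variant of the F_MED_MelodyUrchin face accessory, defining physics for two braids and a hair chain.
# Capsule chains hang from small master spheres on the head, with the braid masters rotationally locked and the hair master free.
# Braid links have narrow, mostly single-axis swing and minimal twist, while hair links are looser at the root.
# Head, upper spine, clavicle and upper-arm capsules form the body shapes, linked by free-rotation constraints.
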PhysicsAsset /BRCosmetics/Characters/Player/Female/Medium/Bodies/F_MED_MelodyUrchin/Meshes/Parts/F_MED_MelodyUrchin_FaceAcc_Frontend_Physics
bodies {
 clavicle_l [{shapeType=Capsule, length=12, radius=12, shapeName=clavicle_l_capsule, center=<3.0958 2 -6.4247>, rotation=<rot 80 9.9984 -0.0018>}]
 clavicle_r [{shapeType=Capsule, length=12, radius=12, shapeName=clavicle_r_capsule, center=<-3.0958 -2 6.4247>, rotation=<rot 80 9.9983 -0.0018>}]
 dyn_braid_l_01 [{shapeType=Capsule, length=3, radius=4, shapeName=dyn_braid_l_01_capsule, center=<2 -2 0>, rotation=<rot 90 0 0>}]
 dyn_braid_l_02 [{shapeType=Capsule, length=3, radius=4, shapeName=dyn_braid_l_02_capsule, center=<2 -2 0>, rotation=<rot 90 0 0>}]
 dyn_braid_l_03 [{shapeType=Capsule, length=3, radius=4, shapeName=dyn_braid_l_03_capsule, center=<2 -2.5 0>, rotation=<rot 90 0 0>}]
 dyn_braid_l_05 [{shapeType=Capsule, length=3, radius=4, shapeName=dyn_braid_l_05_capsule, center=<2 -2.5 0>, rotation=<rot 90 0 0>}]
 dyn_braid_l_06 [{shapeType=Capsule, length=3, radius=4, shapeName=dyn_braid_l_06_capsule, center=<2 -2.5 0>, rotation=<rot 90 0 0>}]
 dyn_braid_l_MSTR [{shapeType=Sphere, radius=0.505, shapeName=dyn_braid_l_MSTR_sphere}]
 dyn_braid_r_01 [{shapeType=Capsule, length=3, radius=4, shapeName=dyn_braid_r_01_capsule, center=<2 -2 0>, rotation=<rot 90 0 0>}]
 dyn_braid_r_02 [{shapeType=Capsule, length=3, radius=4, shapeName=dyn_braid_r_02_capsule, center=<2 -2 0>, rotation=<rot 90 0 0>}]
 dyn_braid_r_03 [{shapeType=Capsule, length=3, radius=4, shapeName=dyn_braid_r_03_capsule, center=<2 -2.5 0>, rotation=<rot 90 0 0>}]
 dyn_braid_r_05 [{shapeType=Capsule, length=3, radius=4, shapeName=dyn_braid_r_05_capsule, center=<2 -2.5 0>, rotation=<rot 90 0 0>}]
 dyn_braid_r_06 [{shapeType=Capsule, length=3, radius=4, shapeName=dyn_braid_r_06_capsule, center=<2 -2.5 0>, rotation=<rot 90 0 0>}]
 dyn_braid_r_MSTR [{shapeType=Sphere, radius=0.505, shapeName=dyn_braid_r_MSTR_sphere}]
 dyn_hair_01 [{shapeType=Capsule, length=5, radius=2.5, shapeName=dyn_hair_01_capsule, center=<1 0.5 0>}]
 dyn_hair_02 [{shapeType=Capsule, length=5, radius=2.5, shapeName=dyn_hair_02_capsule, center=<1 0.5 0>}]
 dyn_hair_03 [{shapeType=Capsule, length=5, radius=2.5, shapeName=dyn_hair_03_capsule, center=<1 0.5 0>}]
 dyn_hair_04 [{shapeType=Capsule, length=5, radius=2.5, shapeName=dyn_hair_04_capsule, center=<1 0.5 0>}]
 dyn_hair_MSTR [{shapeType=Sphere, radius=0.505, shapeName=dyn_hair_MSTR_sphere}]
 head [{shapeType=Capsule, length=6, radius=7, shapeName=head_capsule, center=<3.1543 1.0099 0>, rotation=<rot 90 0 10.0002>}]
 spine_04 [{shapeType=Capsule, length=16, radius=11, shapeName=spine_04_capsule, center=<0 0.5 0>}]
 spine_05 [{shapeType=Capsule, length=16, radius=11, shapeName=spine_05_capsule, center=<0 1.5 0>}]
 upperarm_l [{shapeType=Capsule, length=20, radius=9, shapeName=upperarm_l_capsule, center=<12 4 -3>, rotation=<rot 90 0 0>}]
 upperarm_r [{shapeType=Capsule, length=20, radius=9, shapeName=upperarm_r_capsule, center=<-12 -4 3>, rotation=<rot 90 0 0>}]
}
constraints [
 {bone1Name=dyn_braid_l_01, bone2Name=dyn_braid_l_MSTR, swing1LimitDegrees=15, swing2LimitDegrees=3, twistLimitDegrees=1}
 {bone1Name=dyn_braid_l_02, bone2Name=dyn_braid_l_01, swing1LimitDegrees=20, swing2LimitDegrees=3, twistLimitDegrees=1}
 {bone1Name=dyn_braid_l_03, bone2Name=dyn_braid_l_02, swing1LimitDegrees=15, swing2LimitDegrees=3, twistLimitDegrees=1}
 {bone1Name=dyn_braid_l_05, bone2Name=dyn_braid_l_03, swing1LimitDegrees=20, swing2LimitDegrees=3, twistLimitDegrees=1}
 {bone1Name=dyn_braid_l_06, bone2Name=dyn_braid_l_05, swing1LimitDegrees=15, swing2LimitDegrees=3, twistLimitDegrees=1}
 {bone1Name=dyn_braid_r_01, bone2Name=dyn_braid_r_MSTR, swing1LimitDegrees=15, swing2LimitDegrees=3, twistLimitDegrees=1}
 {bone1Name=dyn_braid_r_02, bone2Name=dyn_braid_r_01, swing1LimitDegrees=20, swing2LimitDegrees=3, twistLimitDegrees=1}
 {bone1Name=dyn_braid_r_03, bone2Name=dyn_braid_r_02, swing1LimitDegrees=15, swing2LimitDegrees=3, twistLimitDegrees=1}
 {bone1Name=dyn_braid_r_05, bone2Name=dyn_braid_r_03, swing1LimitDegrees=20, swing2LimitDegrees=3, twistLimitDegrees=1}
 {bone1Name=dyn_braid_r_06, bone2Name=dyn_braid_r_05, swing1LimitDegrees=15, swing2LimitDegrees=3, twistLimitDegrees=1}
 {bone1Name=dyn_hair_01, bone2Name=dyn_hair_MSTR, swing1LimitDegrees=45, swing2LimitDegrees=10, twistLimitDegrees=20}
 {bone1Name=dyn_hair_02, bone2Name=dyn_hair_01, swing1LimitDegrees=10, swing2LimitDegrees=10, twistLimitDegrees=4}
 {bone1Name=dyn_hair_03, bone2Name=dyn_hair_02, swing1LimitDegrees=10, swing2LimitDegrees=10, twistLimitDegrees=4}
 {bone1Name=dyn_hair_04, bone2Name=dyn_hair_03, swing1LimitDegrees=10, swing2LimitDegrees=10, twistLimitDegrees=4}
 {
  bone1Name=dyn_hair_MSTR
  bone2Name=head
  swing1LimitDegrees=45
  swing1Motion=Free
  swing2LimitDegrees=45
  swing2Motion=Free
  twistLimitDegrees=45
  twistMotion=Free
 }
 {
  bone1Name=dyn_braid_r_MSTR
  bone2Name=head
  swing1LimitDegrees=45
  swing1Motion=Locked
  swing2LimitDegrees=45
  swing2Motion=Locked
  twistLimitDegrees=45
  twistMotion=Locked
 }
 {
  bone1Name=dyn_braid_l_MSTR
  bone2Name=head
  swing1LimitDegrees=45
  swing1Motion=Locked
  swing2LimitDegrees=45
  swing2Motion=Locked
  twistLimitDegrees=45
  twistMotion=Locked
 }
 {
  bone1Name=head
  bone2Name=spine_05
  swing1LimitDegrees=45
  swing1Motion=Free
  swing2LimitDegrees=45
  swing2Motion=Free
  twistLimitDegrees=45
  twistMotion=Free
 }
 {
  bone1Name=clavicle_l
  bone2Name=spine_05
  swing1LimitDegrees=45
  swing1Motion=Free
  swing2LimitDegrees=45
  swing2Motion=Free
  twistLimitDegrees=45
  twistMotion=Free
 }
 {
  bone1Name=clavicle_r
  bone2Name=spine_05
  swing1LimitDegrees=45
  swing1Motion=Free
  swing2LimitDegrees=45
  swing2Motion=Free
  twistLimitDegrees=45
  twistMotion=Free
 }
 {
  bone1Name=spine_05
  bone2Name=spine_04
  swing1LimitDegrees=45
  swing1Motion=Free
  swing2LimitDegrees=45
  swing2Motion=Free
  twistLimitDegrees=45
  twistMotion=Free
 }
 {
  bone1Name=upperarm_r
  bone2Name=clavicle_r
  swing1LimitDegrees=45
  swing1Motion=Free
  swing2LimitDegrees=45
  swing2Motion=Free
  twistLimitDegrees=45
  twistMotion=Free
 }
 {
  bone1Name=upperarm_l
  bone2Name=clavicle_l
  swing1LimitDegrees=45
  swing1Motion=Free
  swing2LimitDegrees=45
  swing2Motion=Free
  twistLimitDegrees=45
  twistMotion=Free
 }
]
