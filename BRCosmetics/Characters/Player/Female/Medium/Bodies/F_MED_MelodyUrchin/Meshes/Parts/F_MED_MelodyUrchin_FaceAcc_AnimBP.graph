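# Post-process style AnimBP for the MelodyUrchin character's face accessory part, adding secondary motion to hair, braids, bangs, earrings and nose ring.
# Parent body and head mesh poses are layered at the head, run through a Control Rig and cached.
# RigidBody sims branch on backbling and frontend state, with the backbling branch also offsetting and squashing hair and braid bones.
# Chained AnimDynamics nodes then swing jewelry and bangs.
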AnimBlueprint /BRCosmetics/Characters/Player/Female/Medium/Bodies/F_MED_MelodyUrchin/Meshes/Parts/F_MED_MelodyUrchin_FaceAcc_AnimBP
target_skeleton=&/F_MED_MelodyUrchin_FaceAcc_Skeleton
graphs [
 AnimGraph {nodes [
   n1 = RigidBody!(ComponentPose=Local To Component(LocalPose=Use cached pose 'SavedPose18778'()), Alpha, OverrideWorldGravity:Vector="0,0,0", ExternalForce) -> (Pose) {
    bOverrideWorldGravity=true
    overridePhysicsAsset=&/F_MED_MelodyUrchin_FaceAcc_Frontend_Physics
    simulationSpace=BaseBoneSpace
    alphaBoolBlend {blendOption=Linear}
    baseBoneRef {boneName=spine_05}
    componentAppliedLinearAccClamp=<3000 3000 3000>
    componentLinearAccScale=<0.4 0.2 0.2>
    componentLinearVelScale=<0.1 0.1 0.1>
    simSpaceSettings {
     maxAngularAcceleration=100
     maxAngularVelocity=3
     maxLinearAcceleration=700
     maxLinearVelocity=500
     velocityScaleZ=0
     worldAlpha=1
     externalLinearDragV=<0.2 0.2>
    }
   }
   n2 = RigidBody(ComponentPose=Local To Component(LocalPose=Use cached pose 'SavedPose18778'()), OverrideWorldGravity="0,0,0") {
    bOverrideWorldGravity=true
    overridePhysicsAsset=&/F_MED_MelodyUrchin_FaceAcc_Physics
    simulationSpace=BaseBoneSpace
    alphaBoolBlend {blendOption=Linear}
    baseBoneRef {boneName=spine_05}
    componentAppliedLinearAccClamp=<3000 3000 3000>
    componentLinearAccScale=<0.4 0.2 0.2>
    componentLinearVelScale=<0.1 0.1 0.1>
    simSpaceSettings {
     maxAngularAcceleration=100
     maxAngularVelocity=3
     maxLinearAcceleration=700
     maxLinearVelocity=500
     velocityScaleZ=0
     worldAlpha=1
     externalLinearDragV=<0.2 0.2>
    }
   }
   n3 = Blend Poses by bool(BlendPose_0=Component To Local(ComponentPose=@n1), BlendPose_1=Component To Local(ComponentPose=@n2), bActiveValue=Get bIsInFrontEnd()) {blendType=Linear}
   n4 = RigidBody(ComponentPose=Local To Component(LocalPose=Use cached pose 'SavedPose18778'()), OverrideWorldGravity="0,0,0") {
    bOverrideWorldGravity=true
    overridePhysicsAsset=&/F_MED_MelodyUrchin_FaceAcc_Backbling_Physics
    simulationSpace=BaseBoneSpace
    alphaBoolBlend {blendOption=Linear}
    baseBoneRef {boneName=spine_05}
    componentAppliedLinearAccClamp=<3000 3000 3000>
    componentLinearAccScale=<0.4 0.2 0.2>
    componentLinearVelScale=<0.1 0.1 0.1>
    simSpaceSettings {
     maxAngularAcceleration=100
     maxAngularVelocity=3
     maxLinearAcceleration=700
     maxLinearVelocity=500
     velocityScaleZ=0
     worldAlpha=1
     externalLinearDragV=<0.2 0.2>
    }
   }
   n5 = "Transform (Modify) Bone - Bone: dyn_hair_03"(ComponentPose=@n4, Alpha=1, Translation="0,1,0", Rotation="0,0,-10", Scale="1,0.75,1") {
    rotationMode=BMM_Additive
    scaleMode=BMM_Replace
    translationMode=BMM_Additive
    alphaBoolBlend {blendOption=Linear}
    boneToModify {boneName=dyn_hair_03}
    rotation {roll=-10}
    scale {y=0.75}
    translation {y=1}
   }
   n6 = "Transform (Modify) Bone - Bone: dyn_hair_04"(ComponentPose=@n5, Alpha=1, Translation="0,1.1,0", Rotation="0,0,-20", Scale="1,0.75,1") {
    rotationMode=BMM_Additive
    scaleMode=BMM_Replace
    translationMode=BMM_Additive
    alphaBoolBlend {blendOption=Linear}
    boneToModify {boneName=dyn_hair_04}
    rotation {roll=-20}
    scale {y=0.75}
    translation {y=1.1}
   }
   n7 = "To Float (Boolean)"(InBool=Get bIsInFrontEnd())
   n8 = "Transform (Modify) Bone - Bone: dyn_hair_01"(ComponentPose=@n6, Alpha=@n7, Translation="0,1,0", Rotation="0,0,0", Scale="1,1,1") {
    rotationMode=BMM_Additive
    scaleMode=BMM_Replace
    translationMode=BMM_Additive
    alphaBoolBlend {blendOption=Linear}
    boneToModify {boneName=dyn_hair_01}
    translation {y=1}
   }
   n9 = "Transform (Modify) Bone - Bone: dyn_hair_02"(ComponentPose=@n8, Alpha=@n7, Translation="0,1,0", Rotation="0,0,0", Scale="1,1,1") {
    rotationMode=BMM_Additive
    scaleMode=BMM_Replace
    translationMode=BMM_Additive
    alphaBoolBlend {blendOption=Linear}
    boneToModify {boneName=dyn_hair_02}
    translation {y=1}
   }
   n10 = "Transform (Modify) Bone - Bone: dyn_hair_03"(ComponentPose=@n9, Alpha=@n7, Translation="0,0,0", Rotation="0,0,5", Scale="1,1,1") {
    rotationMode=BMM_Additive
    scaleMode=BMM_Replace
    translationMode=BMM_Additive
    alphaBoolBlend {blendOption=Linear}
    boneToModify {boneName=dyn_hair_03}
    rotation {roll=5}
   }
   n11 = "Transform (Modify) Bone - Bone: dyn_braid_r_02"(ComponentPose=@n10, Alpha=1, Translation="0,1,0", Rotation="0,0,0", Scale="1,0.8,1") {
    rotationMode=BMM_Additive
    scaleMode=BMM_Replace
    translationMode=BMM_Additive
    alphaBoolBlend {blendOption=Linear}
    boneToModify {boneName=dyn_braid_r_02}
    scale {y=0.8}
    translation {y=1}
   }
   n12 = "Transform (Modify) Bone - Bone: dyn_braid_r_03"(ComponentPose=@n11, Alpha=1, Translation="0,0.5,0", Rotation="0,0,0", Scale="1,0.75,1") {
    rotationMode=BMM_Additive
    scaleMode=BMM_Replace
    translationMode=BMM_Additive
    alphaBoolBlend {blendOption=Linear}
    boneToModify {boneName=dyn_braid_r_03}
    scale {y=0.75}
    translation {y=0.5}
   }
   n13 = "Transform (Modify) Bone - Bone: dyn_braid_r_04"(ComponentPose=@n12, Alpha=1, Translation="0,0.5,0", Rotation="0,0,0", Scale="1,0.5,1") {
    rotationMode=BMM_Additive
    scaleMode=BMM_Replace
    translationMode=BMM_Additive
    alphaBoolBlend {blendOption=Linear}
    boneToModify {boneName=dyn_braid_r_04}
    scale {y=0.5}
    translation {y=0.5}
   }
   n14 = "Transform (Modify) Bone - Bone: dyn_braid_r_05"(ComponentPose=@n13, Alpha=1, Translation="0,0.8,0", Rotation="-30,0,15", Scale="1,0.25,1") {
    rotationMode=BMM_Additive
    scaleMode=BMM_Replace
    translationMode=BMM_Additive
    alphaBoolBlend {blendOption=Linear}
    boneToModify {boneName=dyn_braid_r_05}
    rotation {pitch=-30, roll=15}
    scale {y=0.25}
    translation {y=0.8}
   }
   n15 = "Transform (Modify) Bone - Bone: dyn_braid_r_05"(ComponentPose=@n14, Alpha=1, Translation="0,0,0", Rotation="0,0,0", Scale="1,0.25,1") {
    rotationMode=BMM_Additive
    scaleMode=BMM_Replace
    translationMode=BMM_Additive
    alphaBoolBlend {blendOption=Linear}
    boneToModify {boneName=dyn_braid_r_05}
    scale {y=0.25}
   }
   n16 = "Transform (Modify) Bone - Bone: dyn_braid_l_02"(ComponentPose=@n15, Alpha=1, Translation="0,1,0", Rotation="0,0,0", Scale="1,0.8,1") {
    scaleMode=BMM_Replace
    translationMode=BMM_Additive
    alphaBoolBlend {blendOption=Linear}
    boneToModify {boneName=dyn_braid_l_02}
    scale {y=0.8}
    translation {y=1}
   }
   n17 = "Transform (Modify) Bone - Bone: dyn_braid_l_03"(ComponentPose=@n16, Alpha=1, Translation="0,0.5,0", Rotation="0,0,0", Scale="1,0.75,1") {
    scaleMode=BMM_Replace
    translationMode=BMM_Additive
    alphaBoolBlend {blendOption=Linear}
    boneToModify {boneName=dyn_braid_l_03}
    scale {y=0.75}
    translation {y=0.5}
   }
   n18 = "Transform (Modify) Bone - Bone: dyn_braid_l_04"(ComponentPose=@n17, Alpha=1, Translation="0,0.5,0", Rotation="0,0,0", Scale="1,0.5,1") {
    scaleMode=BMM_Replace
    translationMode=BMM_Additive
    alphaBoolBlend {blendOption=Linear}
    boneToModify {boneName=dyn_braid_l_04}
    scale {y=0.5}
    translation {y=0.5}
   }
   n19 = "Transform (Modify) Bone - Bone: dyn_braid_l_05"(ComponentPose=@n18, Alpha=1, Translation="0,0.8,0", Rotation="30,0,-15", Scale="1,0.25,1") {
    scaleMode=BMM_Replace
    translationMode=BMM_Additive
    alphaBoolBlend {blendOption=Linear}
    boneToModify {boneName=dyn_braid_l_05}
    rotation {pitch=30, roll=-15}
    scale {y=0.25}
    translation {y=0.8}
   }
   n20 = "Transform (Modify) Bone - Bone: dyn_braid_l_05"(ComponentPose=@n19, Alpha=1, Translation="0,0,0", Rotation="0,0,0", Scale="1,0.25,1") {
    scaleMode=BMM_Replace
    translationMode=BMM_Additive
    alphaBoolBlend {blendOption=Linear}
    boneToModify {boneName=dyn_braid_l_05}
    scale {y=0.25}
   }
   n21 = Component To Local(ComponentPose=@n20)
   n22 = Get IsWearingBackbling() -> (IsWearingBackbling:bool)
   n23 = Anim Dynamics - Bone: dyn_earring_l_01!(
    ComponentPose=Local To Component(LocalPose=Blend Poses by bool(BlendPose_0=@n21, BlendPose_1=@n3, bActiveValue=@n22) {blendType=Linear})
    Alpha=1
    GravityOverride:Vector="0,0,0"
    ExternalForce="0,0,0"
   ) -> (Pose) {
    angularSpringConstant=10
    bUseGravityOverride=true
    simulationSpace=World
    alphaBoolBlend {blendOption=Linear}
    boundBone {boneName=dyn_earring_l_01}
    componentAppliedLinearAccClamp=<3000 3000 3000>
    componentLinearAccScale=<0.2 0.2 0.2>
    componentLinearVelScale=<0.1 0.1 0.1>
    physicsBodyDefinitions [{
      boundBone {boneName=dyn_earring_l_01}
      boxExtents=<8 4 8>
      constraintSetup {angularLimitsMax=<1 3 10>, angularLimitsMin=<-1 -3 -12>, angularTarget=<0 0 0>}
      localJointOffset=<1 0 0>
     }]
    relativeSpaceBone {boneName=spine_05}
   }
   n24 = Anim Dynamics - Bone: dyn_earring_r_01!(ComponentPose=@n23, Alpha=1, GravityOverride:Vector="0,0,0", ExternalForce="0,0,0") -> (Pose) {
    angularSpringConstant=10
    bUseGravityOverride=true
    simulationSpace=World
    alphaBoolBlend {blendOption=Linear}
    boundBone {boneName=dyn_earring_r_01}
    componentAppliedLinearAccClamp=<3000 3000 3000>
    componentLinearAccScale=<0.2 0.2 0.2>
    componentLinearVelScale=<0.1 0.1 0.1>
    physicsBodyDefinitions [{
      boundBone {boneName=dyn_earring_r_01}
      boxExtents=<8 4 8>
      constraintSetup {angularLimitsMax=<1 3 10>, angularLimitsMin=<-1 -3 -12>, angularTarget=<0 0 0>}
      localJointOffset=<1 0 0>
     }]
    relativeSpaceBone {boneName=spine_05}
   }
   n25 = Anim Dynamics - Bone: dyn_nosering_01!(ComponentPose=@n24, Alpha=1, GravityOverride:Vector="0,0,0", ExternalForce="0,0,0") -> (Pose) {
    angularSpringConstant=10
    bUseGravityOverride=true
    simulationSpace=World
    alphaBoolBlend {blendOption=Linear}
    boundBone {boneName=dyn_nosering_01}
    componentAppliedLinearAccClamp=<3000 3000 3000>
    componentLinearAccScale=<0.2 0.2 0.2>
    componentLinearVelScale=<0.1 0.1 0.1>
    physicsBodyDefinitions [{
      boundBone {boneName=dyn_nosering_01}
      boxExtents=<8 4 8>
      constraintSetup {angularLimitsMax=<1 10 3>, angularLimitsMin=<-1 -10 -3>, angularTarget=<0 0 0>}
      localJointOffset=<1 0 0>
     }]
    relativeSpaceBone {boneName=spine_05}
   }
   n26 = Anim Dynamics - Bone: dyn_bang_r_01!(ComponentPose=@n25, Alpha=1, GravityOverride:Vector="0,0,0", ExternalForce="0,0,0") -> (Pose) {
    angularSpringConstant=10
    bAngularSpring=true
    bUseGravityOverride=true
    simulationSpace=World
    alphaBoolBlend {blendOption=Linear}
    boundBone {boneName=dyn_bang_r_01}
    componentAppliedLinearAccClamp=<1000 1000 1000>
    componentLinearAccScale=<0.2 0.2 0.2>
    componentLinearVelScale=<0.1 0.1 0.1>
    physicsBodyDefinitions [{
      boundBone {boneName=dyn_bang_r_01}
      boxExtents=<12 6 12>
      constraintSetup {angularLimitsMax=<1 20 4>, angularLimitsMin=<-1 -5 -4>, angularTarget=<10000 0 0>}
      localJointOffset=<0.8 0 0>
     }]
    relativeSpaceBone {boneName=spine_05}
   }
   n27 = Anim Dynamics - Bone: dyn_bang_r_02!(ComponentPose=@n26, Alpha=1, GravityOverride:Vector="0,0,0", ExternalForce="0,0,0") -> (Pose) {
    angularSpringConstant=10
    bAngularSpring=true
    bUseGravityOverride=true
    simulationSpace=World
    alphaBoolBlend {blendOption=Linear}
    boundBone {boneName=dyn_bang_r_02}
    componentAppliedLinearAccClamp=<1000 1000 1000>
    componentLinearAccScale=<0.2 0.2 0.2>
    componentLinearVelScale=<0.1 0.1 0.1>
    physicsBodyDefinitions [{
      boundBone {boneName=dyn_bang_r_02}
      boxExtents=<12 6 12>
      constraintSetup {angularLimitsMax=<1 20 4>, angularLimitsMin=<-1 -5 -4>, angularTarget=<10000 0 0>}
      localJointOffset=<0.8 0 0>
     }]
    relativeSpaceBone {boneName=spine_05}
   }
   n28 = Anim Dynamics - Bone: dyn_bang_r_03!(ComponentPose=@n27, Alpha=1, GravityOverride:Vector="0,0,0", ExternalForce="0,0,0") -> (Pose) {
    angularSpringConstant=10
    bAngularSpring=true
    bUseGravityOverride=true
    simulationSpace=World
    alphaBoolBlend {blendOption=Linear}
    boundBone {boneName=dyn_bang_r_03}
    componentAppliedLinearAccClamp=<1000 1000 1000>
    componentLinearAccScale=<0.2 0.2 0.2>
    componentLinearVelScale=<0.1 0.1 0.1>
    physicsBodyDefinitions [{
      boundBone {boneName=dyn_bang_r_03}
      boxExtents=<12 6 12>
      constraintSetup {angularLimitsMax=<1 20 4>, angularLimitsMin=<-1 -5 -4>, angularTarget=<10000 0 0>}
      localJointOffset=<0.8 0 0>
     }]
    relativeSpaceBone {boneName=spine_05}
   }
   n29 = Anim Dynamics - Bone: dyn_bang_r_04!(ComponentPose=@n28, Alpha=1, GravityOverride:Vector="0,0,0", ExternalForce="0,0,0") -> (Pose) {
    angularSpringConstant=10
    bAngularSpring=true
    bUseGravityOverride=true
    simulationSpace=World
    alphaBoolBlend {blendOption=Linear}
    boundBone {boneName=dyn_bang_r_04}
    componentAppliedLinearAccClamp=<1000 1000 1000>
    componentLinearAccScale=<0.2 0.2 0.2>
    componentLinearVelScale=<0.1 0.1 0.1>
    physicsBodyDefinitions [{
      boundBone {boneName=dyn_bang_r_04}
      boxExtents=<12 6 12>
      constraintSetup {angularLimitsMax=<1 20 4>, angularLimitsMin=<-1 -5 -4>, angularTarget=<10000 0 0>}
      localJointOffset=<0.8 0 0>
     }]
    relativeSpaceBone {boneName=spine_05}
   }
   n30 = Anim Dynamics - Bone: dyn_bang_c_01!(ComponentPose=@n29, Alpha=1, GravityOverride:Vector="0,0,0", ExternalForce="0,0,0") -> (Pose) {
    angularSpringConstant=10
    bAngularSpring=true
    bUseGravityOverride=true
    simulationSpace=World
    alphaBoolBlend {blendOption=Linear}
    boundBone {boneName=dyn_bang_c_01}
    componentAppliedLinearAccClamp=<1000 1000 1000>
    componentLinearAccScale=<0.2 0.2 0.2>
    componentLinearVelScale=<0.1 0.1 0.1>
    physicsBodyDefinitions [{
      boundBone {boneName=dyn_bang_c_01}
      boxExtents=<12 8 12>
      constraintSetup {angularLimitsMax=<1 4 1>, angularLimitsMin=<-1 -4 -10>, angularTarget=<10000 0 0>}
      localJointOffset=<0.8 0 0>
     }]
    relativeSpaceBone {boneName=spine_05}
   }
   n31 = Anim Dynamics - Bone: dyn_bang_c_02!(ComponentPose=@n30, Alpha=1, GravityOverride:Vector="0,0,0", ExternalForce="0,0,0") -> (Pose) {
    angularSpringConstant=10
    bAngularSpring=true
    bUseGravityOverride=true
    simulationSpace=World
    alphaBoolBlend {blendOption=Linear}
    boundBone {boneName=dyn_bang_c_02}
    componentAppliedLinearAccClamp=<1000 1000 1000>
    componentLinearAccScale=<0.2 0.2 0.2>
    componentLinearVelScale=<0.1 0.1 0.1>
    physicsBodyDefinitions [{
      boundBone {boneName=dyn_bang_c_02}
      boxExtents=<12 8 12>
      constraintSetup {angularLimitsMax=<1 4 1>, angularLimitsMin=<-1 -4 -10>, angularTarget=<10000 0 0>}
      localJointOffset=<0.8 0 0>
     }]
    relativeSpaceBone {boneName=spine_05}
   }
   n32 = Anim Dynamics - Bone: dyn_bang_c_03!(ComponentPose=@n31, Alpha=1, GravityOverride:Vector="0,0,0", ExternalForce="0,0,0") -> (Pose) {
    angularSpringConstant=10
    bAngularSpring=true
    bUseGravityOverride=true
    simulationSpace=World
    alphaBoolBlend {blendOption=Linear}
    boundBone {boneName=dyn_bang_c_03}
    componentAppliedLinearAccClamp=<1000 1000 1000>
    componentLinearAccScale=<0.2 0.2 0.2>
    componentLinearVelScale=<0.1 0.1 0.1>
    physicsBodyDefinitions [{
      boundBone {boneName=dyn_bang_c_03}
      boxExtents=<12 8 12>
      constraintSetup {angularLimitsMax=<1 8 1>, angularLimitsMin=<-1 -8 -10>, angularTarget=<10000 0 0>}
      localJointOffset=<0.8 0 0>
     }]
    relativeSpaceBone {boneName=spine_05}
   }
   n33 = Anim Dynamics - Bone: dyn_bang_c_04!(ComponentPose=@n32, Alpha=1, GravityOverride:Vector="0,0,0", ExternalForce="0,0,0") -> (Pose) {
    angularSpringConstant=10
    bAngularSpring=true
    bUseGravityOverride=true
    simulationSpace=World
    alphaBoolBlend {blendOption=Linear}
    boundBone {boneName=dyn_bang_c_04}
    componentAppliedLinearAccClamp=<1000 1000 1000>
    componentLinearAccScale=<0.2 0.2 0.2>
    componentLinearVelScale=<0.1 0.1 0.1>
    physicsBodyDefinitions [{
      boundBone {boneName=dyn_bang_c_04}
      boxExtents=<12 8 12>
      constraintSetup {angularLimitsMax=<1 8 1>, angularLimitsMin=<-1 -8 -10>, angularTarget=<10000 0 0>}
      localJointOffset=<0.8 0 0>
     }]
    relativeSpaceBone {boneName=spine_05}
   }
   n34 = Anim Dynamics - Bone: dyn_bang_c_05!(ComponentPose=@n33, Alpha=1, GravityOverride:Vector="0,0,0", ExternalForce="0,0,0") -> (Pose) {
    angularSpringConstant=10
    bAngularSpring=true
    bUseGravityOverride=true
    simulationSpace=World
    alphaBoolBlend {blendOption=Linear}
    boundBone {boneName=dyn_bang_c_05}
    componentAppliedLinearAccClamp=<1000 1000 1000>
    componentLinearAccScale=<0.2 0.2 0.2>
    componentLinearVelScale=<0.1 0.1 0.1>
    physicsBodyDefinitions [{
      boundBone {boneName=dyn_bang_c_05}
      boxExtents=<12 8 12>
      constraintSetup {angularLimitsMax=<1 8 1>, angularLimitsMin=<-1 -8 -10>, angularTarget=<10000 0 0>}
      localJointOffset=<0.8 0 0>
     }]
    relativeSpaceBone {boneName=spine_05}
   }
   n35 = Anim Dynamics - Bone: dyn_bang_l_f_01!(ComponentPose=@n34, Alpha=1, GravityOverride:Vector="0,0,0", ExternalForce="0,0,0") -> (Pose) {
    angularSpringConstant=10
    bAngularSpring=true
    bUseGravityOverride=true
    simulationSpace=World
    alphaBoolBlend {blendOption=Linear}
    boundBone {boneName=dyn_bang_l_f_01}
    componentAppliedLinearAccClamp=<1000 1000 1000>
    componentLinearAccScale=<0.2 0.2 0.2>
    componentLinearVelScale=<0.1 0.1 0.1>
    physicsBodyDefinitions [{
      boundBone {boneName=dyn_bang_l_f_01}
      boxExtents=<16 10 16>
      constraintSetup {angularLimitsMax=<1 4 1>, angularLimitsMin=<-1 -4 -10>, angularTarget=<10000 0 0>}
      localJointOffset=<0.9 0 0>
     }]
    relativeSpaceBone {boneName=spine_05}
   }
   n36 = Anim Dynamics - Bone: dyn_bang_l_f_02!(ComponentPose=@n35, Alpha=1, GravityOverride:Vector="0,0,0", ExternalForce="0,0,0") -> (Pose) {
    angularSpringConstant=10
    bAngularSpring=true
    bUseGravityOverride=true
    simulationSpace=World
    alphaBoolBlend {blendOption=Linear}
    boundBone {boneName=dyn_bang_l_f_02}
    componentAppliedLinearAccClamp=<1000 1000 1000>
    componentLinearAccScale=<0.2 0.2 0.2>
    componentLinearVelScale=<0.1 0.1 0.1>
    physicsBodyDefinitions [{
      boundBone {boneName=dyn_bang_l_f_02}
      boxExtents=<16 10 16>
      constraintSetup {angularLimitsMax=<1 4 1>, angularLimitsMin=<-1 -4 -10>, angularTarget=<10000 0 0>}
      localJointOffset=<0.9 0 0>
     }]
    relativeSpaceBone {boneName=spine_05}
   }
   n37 = Anim Dynamics - Bone: dyn_bang_l_f_03!(ComponentPose=@n36, Alpha=1, GravityOverride:Vector="0,0,0", ExternalForce="0,0,0") -> (Pose) {
    angularSpringConstant=10
    bAngularSpring=true
    bUseGravityOverride=true
    simulationSpace=World
    alphaBoolBlend {blendOption=Linear}
    boundBone {boneName=dyn_bang_l_f_03}
    componentAppliedLinearAccClamp=<1000 1000 1000>
    componentLinearAccScale=<0.2 0.2 0.2>
    componentLinearVelScale=<0.1 0.1 0.1>
    physicsBodyDefinitions [{
      boundBone {boneName=dyn_bang_l_f_03}
      boxExtents=<16 10 16>
      constraintSetup {angularLimitsMax=<1 12 1>, angularLimitsMin=<-1 -24 -10>, angularTarget=<10000 0 0>}
      localJointOffset=<0.9 0 0>
     }]
    relativeSpaceBone {boneName=spine_05}
   }
   n38 = Anim Dynamics - Bone: dyn_bang_l_f_04!(ComponentPose=@n37, Alpha=1, GravityOverride:Vector="0,0,0", ExternalForce="0,0,0") -> (Pose) {
    angularSpringConstant=10
    bAngularSpring=true
    bUseGravityOverride=true
    simulationSpace=World
    alphaBoolBlend {blendOption=Linear}
    boundBone {boneName=dyn_bang_l_f_04}
    componentAppliedLinearAccClamp=<1000 1000 1000>
    componentLinearAccScale=<0.2 0.2 0.2>
    componentLinearVelScale=<0.1 0.1 0.1>
    physicsBodyDefinitions [{
      boundBone {boneName=dyn_bang_l_f_04}
      boxExtents=<16 10 16>
      constraintSetup {angularLimitsMax=<1 12 1>, angularLimitsMin=<-1 -24 -10>, angularTarget=<10000 0 0>}
      localJointOffset=<0.9 0 0>
     }]
    relativeSpaceBone {boneName=spine_05}
   }
   n39 = Anim Dynamics - Bone: dyn_bang_l_f_05!(ComponentPose=@n38, Alpha=1, GravityOverride:Vector="0,0,0", ExternalForce="0,0,0") -> (Pose) {
    angularSpringConstant=10
    bAngularSpring=true
    bUseGravityOverride=true
    simulationSpace=World
    alphaBoolBlend {blendOption=Linear}
    boundBone {boneName=dyn_bang_l_f_05}
    componentAppliedLinearAccClamp=<1000 1000 1000>
    componentLinearAccScale=<0.2 0.2 0.2>
    componentLinearVelScale=<0.1 0.1 0.1>
    physicsBodyDefinitions [{
      boundBone {boneName=dyn_bang_l_f_05}
      boxExtents=<16 10 16>
      constraintSetup {angularLimitsMax=<1 12 1>, angularLimitsMin=<-1 -24 -10>, angularTarget=<10000 0 0>}
      localJointOffset=<0.9 0 0>
     }]
    relativeSpaceBone {boneName=spine_05}
   }
   n40 = Anim Dynamics - Bone: dyn_bang_l_b_01!(ComponentPose=@n39, Alpha=1, GravityOverride:Vector="0,0,0", ExternalForce="0,0,0") -> (Pose) {
    angularSpringConstant=10
    bAngularSpring=true
    bUseGravityOverride=true
    simulationSpace=World
    alphaBoolBlend {blendOption=Linear}
    boundBone {boneName=dyn_bang_l_b_01}
    componentAppliedLinearAccClamp=<1000 1000 1000>
    componentLinearAccScale=<0.2 0.2 0.2>
    componentLinearVelScale=<0.1 0.1 0.1>
    physicsBodyDefinitions [{
      boundBone {boneName=dyn_bang_l_b_01}
      boxExtents=<14 7 14>
      constraintSetup {angularLimitsMax=<1 10 2>, angularLimitsMin=<-1 -20 -6>, angularTarget=<10000 0 0>}
      localJointOffset=<0.9 0 0>
     }]
    relativeSpaceBone {boneName=spine_05}
   }
   n41 = Anim Dynamics - Bone: dyn_bang_l_b_02!(ComponentPose=@n40, Alpha=1, GravityOverride:Vector="0,0,0", ExternalForce="0,0,0") -> (Pose) {
    angularSpringConstant=10
    bAngularSpring=true
    bUseGravityOverride=true
    simulationSpace=World
    alphaBoolBlend {blendOption=Linear}
    boundBone {boneName=dyn_bang_l_b_02}
    componentAppliedLinearAccClamp=<1000 1000 1000>
    componentLinearAccScale=<0.2 0.2 0.2>
    componentLinearVelScale=<0.1 0.1 0.1>
    physicsBodyDefinitions [{
      boundBone {boneName=dyn_bang_l_b_02}
      boxExtents=<14 7 14>
      constraintSetup {angularLimitsMax=<1 10 2>, angularLimitsMin=<-1 -20 -6>, angularTarget=<10000 0 0>}
      localJointOffset=<0.9 0 0>
     }]
    relativeSpaceBone {boneName=spine_05}
   }
   n42 = Anim Dynamics - Bone: dyn_bang_l_b_03!(ComponentPose=@n41, Alpha=1, GravityOverride:Vector="0,0,0", ExternalForce="0,0,0") -> (Pose) {
    angularSpringConstant=10
    bAngularSpring=true
    bUseGravityOverride=true
    simulationSpace=World
    alphaBoolBlend {blendOption=Linear}
    boundBone {boneName=dyn_bang_l_b_03}
    componentAppliedLinearAccClamp=<1000 1000 1000>
    componentLinearAccScale=<0.2 0.2 0.2>
    componentLinearVelScale=<0.1 0.1 0.1>
    physicsBodyDefinitions [{
      boundBone {boneName=dyn_bang_l_b_03}
      boxExtents=<14 7 14>
      constraintSetup {angularLimitsMax=<1 10 2>, angularLimitsMin=<-1 -20 -6>, angularTarget=<10000 0 0>}
      localJointOffset=<0.9 0 0>
     }]
    relativeSpaceBone {boneName=spine_05}
   }
   n43 = Anim Dynamics - Bone: dyn_bang_l_01!(ComponentPose=@n42, Alpha=1, GravityOverride:Vector="0,0,0", ExternalForce="0,0,0") -> (Pose) {
    angularSpringConstant=10
    bAngularSpring=true
    bUseGravityOverride=true
    simulationSpace=World
    alphaBoolBlend {blendOption=Linear}
    boundBone {boneName=dyn_bang_l_01}
    componentAppliedLinearAccClamp=<1000 1000 1000>
    componentLinearAccScale=<0.2 0.2 0.2>
    componentLinearVelScale=<0.1 0.1 0.1>
    physicsBodyDefinitions [{
      boundBone {boneName=dyn_bang_l_01}
      boxExtents=<12 6 12>
      constraintSetup {angularLimitsMax=<1 10 2>, angularLimitsMin=<-1 -20 -6>, angularTarget=<10000 0 0>}
      localJointOffset=<1 0 0>
     }]
    relativeSpaceBone {boneName=spine_05}
   }
   n44 = Anim Dynamics - Bone: dyn_bang_l_02!(ComponentPose=@n43, Alpha=1, GravityOverride:Vector="0,0,0", ExternalForce="0,0,0") -> (Pose) {
    angularSpringConstant=10
    bAngularSpring=true
    bUseGravityOverride=true
    simulationSpace=World
    alphaBoolBlend {blendOption=Linear}
    boundBone {boneName=dyn_bang_l_02}
    componentAppliedLinearAccClamp=<1000 1000 1000>
    componentLinearAccScale=<0.2 0.2 0.2>
    componentLinearVelScale=<0.1 0.1 0.1>
    physicsBodyDefinitions [{
      boundBone {boneName=dyn_bang_l_02}
      boxExtents=<12 6 12>
      constraintSetup {angularLimitsMax=<1 10 2>, angularLimitsMin=<-1 -20 -6>, angularTarget=<10000 0 0>}
      localJointOffset=<1 0 0>
     }]
    relativeSpaceBone {boneName=spine_05}
   }
   Output Pose(Result=Component To Local(ComponentPose=@n44)) {layerGroup=None}
   n45 = Layered blend per bone(
    BasePose=Copy Pose From Mesh() {bUseAttachedParent=true}
    BlendPoses_0=Copy Pose From Mesh(SourceMeshComponent=Get Head Mesh() -> (Head Mesh:Skeletal Mesh Component Object Reference))
   ) {layerSetup [{branchFilters [{boneName=head}, {boneName=FACIAL_C_Nose}, {boneName=FACIAL_C_NoseLower}]}]}
   n46 = Control Rig(Source=@n45) {
    alphaBoolBlend {blendOption=Linear}
    controlRigAssetReference {blueprintRigClass=&/F_MED_MelodyUrchin_FaceAcc_CtrlRig.F_MED_MelodyUrchin_FaceAcc_CtrlRig_C}
   }
   Save cached pose 'SavedPose18778'(Pose=@n46) {cacheName=SavedPose18778}
  ]}
 EventGraph {nodes [
   Event BlueprintUpdateAnimation()
   > n1 = Gravity Override State Function DEPRECATED(CharacterPartAnimInstance=Self-Reference(), GameTimeInSeconds=GetGameTimeInSeconds(), GravityOverrideParams=Get Gravity Override Params())
   > Set Gravity Override(Gravity Override=@n1)
   > n2 = Cast To PlayerPawn_Athena(Object=TryGetPawnOwner())
   > Set Head Mesh(Head Mesh:Skeletal Mesh Component Object Reference=GetSkeletalMeshForPartType(self=@n2, PartType=Head)) -> (Output_Get:Skeletal Mesh Component Object Reference)
   n3 = CheckForBackbling()
   > Set IsWearingBackbling(IsWearingBackbling:bool=@n3) -> (Output_Get:bool)
   AND Boolean(A=@n3, B=NOT Boolean(A=@n3.bIsBackblingACape))
  ]}
]
variables [
 Gravity Override {type=Vector}
 Gravity Override Params {type=GravityOverrideParamsStruct}
 Head Mesh {type=SkeletalMeshComponent}
 IsWearingBackbling {type=boolean}
]
